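# Sound cue for the closing sound of the Creative sliding door gadget.
# A single wave player plays one close sample through a modulator node, with slightly raised pitch and reduced volume on the cue.
# It uses the Creative doors attenuation and the doors sound class, and overrides concurrency to prevent new instances rather than cutting off playing ones.
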SoundCue /Game/Sounds/Creative/Gadgets/SlidingDoor/SlidingDoor_Close_Cue
root=Modulator_0
nodes [
 {id=Modulator_0, children [WavePlayer_1]}
 {id=WavePlayer_1, properties {soundWaveAssetPtr=&/SlidingDoor_Close_01}}
]
properties {
 attenuationSettings=/Game/Sounds/Creative/Attenuations/Fort_Creative_Doors_Att
 bOverrideConcurrency=true
 maxDistance=2400
 pitchMultiplier=1.05
 soundClassObject=&Fort_Doors
 volumeMultiplier=0.7
 concurrencyOverrides {resolutionRule=PreventNew}
}
